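# Niagara sprite effect for the TrafficHat emote: short-lived drops, tinted cyan or pink by spawn group, travel from tube sockets toward a vial socket.
# Two staggered bursts spawn particles that cache both socket positions.
# Update sets position directly from a curve-driven lerp between them, shaped by per-axis curves, plus per-group offsets blended by User.DropOffset.
# Curves grow and stretch the sprites and blend their alignment toward vertical.
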NiagaraSystem /BRCosmetics/Animation/Game/MainPlayer/Emotes/TrafficHat/FX/NS_Emote_TrafficHat_Drops
dependencies {
 usedDataInterfaces [
  &NiagaraDataInterfaceCurve
  &NiagaraDataInterfaceVectorCurve
  &NiagaraDataInterfaceVector2DCurve
  &NiagaraDataInterfaceSocketReader
 ]
 usedDynamicInputs []
 usedModules [
  &self:SystemSpawnScript.SystemScriptSource.SystemScriptGraph.NiagaraNodeAssignment_1.SetVariables_#1
  &self:SystemSpawnScript.SystemScriptSource.SystemScriptGraph.NiagaraNodeAssignment_2.SetVariables_#2
  &SystemState
 ]
}
emitters {Minimal {
  emitterUpdateScript [
   &EmitterState(Life Cycle Mode=::NewEnumerator1, Loop Behavior=::NewEnumerator1, Loop Duration=2, UseLoopDelay=false)
   &SpawnBurst_Instantaneous(Use Spawn Probability=false, UseLoopCountLimit=false)
   &SpawnBurst_Instantaneous(Spawn Time=0.09, Use Spawn Probability=false, UseLoopCountLimit=false) {module=SpawnBurst_Instantaneous001}
  ]
  particleSpawnScript [
   &FN_SetSpawnGroup(Number of Spawn Groups=2)
   &InitializeParticle(
    Color=&MaskLinearColorBySpawnGroup(Masked Linear Color=<rgba 0.1189 0.6795 1 1>, Passthrough Linear Color=<rgba 1 0.158 0.8341 1>)
    Lifetime=0.5
    Sprite Rotation Angle=0.5
    Sprite Rotation Mode=::NewEnumerator1
    Sprite Size=<2 4>
    Sprite Size Mode=::NewEnumerator4
    UsePositionOffset=false
    Write Material Random=true
   )
   &AddVelocity(Velocity=<0 0 -0.001>) {module=AddVelocity001}
   &SocketLocation(
    SamplePosOffset=false
    Socket Index=&SpawnGroup()
    Socket Reader=$System.TubeLocations
    Socket Sampling Mode=::NewEnumerator3
    VelocitySpeedLimit=false
   )
   &SocketLocation(SamplePosOffset=false, Socket Reader=$System.VialLocation, Socket Sampling Mode=::NewEnumerator3, VelocitySpeedLimit=false) {module=SocketLocation001}
   /SetVariables(
    Particles.TubeLocations:NiagaraPosition=$Output.SocketLocation.SampledPosition
    Particles.VialLocations:NiagaraPosition=$Output.SocketLocation001.SampledPosition
   ) {isSetParameters=true}
  ]
  particleUpdateScript [
   &ParticleState()
   /SetVariables(Particles.Position=&AddVectorToPosition(
     Position=&Lerp_Vector(
      A=&MaskVectorBySpawnGroup(Masked Vector=<-0.5 0 9>, Passthrough Vector=<-0.5 0 14>)
      B=&MaskVectorBySpawnGroup(Masked Vector=<1 0 4>, Passthrough Vector=<0 0 10>)
      Alpha=$User.DropOffset
     )
     Vector=&Multiply_Vector(
      A=&LerpPosition(
       A=Linked() {linkedVariable {name=Particles.TubeLocations, type=Vector3f}}
       B=Linked() {linkedVariable {name=Particles.VialLocations, type=Vector3f}}
       Alpha=&FloatFromCurve(FloatCurve=DataInterface() {propertyValues {Curve {keys [
            {interpMode=RCIM_Cubic, tangentMode=RCTM_User, time=0, value=0}
            {arriveTangent=3.3274, interpMode=RCIM_Cubic, leaveTangent=3.3274, tangentMode=RCTM_User, time=1, value=1}
           ]}}})
      )
      B=&VectorFromCurve(VectorCurve=DataInterface() {propertyValues {
         XCurve {keys [
           {arriveTangent=-3.4398, interpMode=RCIM_Cubic, leaveTangent=-3.4398, leaveTangentWeight=0.3483, tangentMode=RCTM_User, time=0, value=1}
           {arriveTangent=-1, interpMode=RCIM_Cubic, leaveTangent=-1, time=0.4509, value=0.3643}
           {interpMode=RCIM_Cubic, tangentMode=RCTM_User, time=1, value=0}
          ]}
         YCurve {keys [{tangentMode=RCTM_User, time=0, value=1}, {tangentMode=RCTM_User, time=1, value=1}]}
         ZCurve {keys [
           {arriveTangent=-2.14, interpMode=RCIM_Cubic, leaveTangent=-2.14, tangentMode=RCTM_User, time=0, value=0.2347}
           {interpMode=RCIM_Cubic, tangentMode=RCTM_User, time=1, value=1}
          ]}
        }})
     )
    )) {isSetParameters=true}
   &SolveForcesAndVelocity(Clamp Velocity=false, Limit Acceleration=false)
   &ScaleSpriteSize(
    Non-Uniform Curve Scale=&Multiply_Vector2D_ByFloat(
     Float=&FloatFromCurve(FloatCurve=DataInterface() {propertyValues {Curve {keys [
          {arriveTangent=6.5967, interpMode=RCIM_Cubic, leaveTangent=6.5967, tangentMode=RCTM_User, time=0, value=0}
          {interpMode=RCIM_Cubic, tangentMode=RCTM_User, time=0.2271, value=1}
          {tangentMode=RCTM_User, time=1, value=1}
         ]}}})
     Vector2D=<1 1>
    )
    Non-Uniform Curve Sprite Scale=DataInterface() {propertyValues {
      XCurve {keys [
        {arriveTangent=4.1785, interpMode=RCIM_Cubic, leaveTangent=4.1785, tangentMode=RCTM_User, time=0, value=0}
        {arriveTangent=1.1179, interpMode=RCIM_Cubic, leaveTangent=1.1179, tangentMode=RCTM_User, time=0.1, value=1.375}
        {arriveTangent=0.0224, interpMode=RCIM_Cubic, leaveTangent=0.0224, tangentMode=RCTM_User, time=1, value=1}
       ]}
      YCurve {keys [
        <at 0 0.2781 RCIM_Cubic>
        {arriveTangentWeight=0.5187, interpMode=RCIM_Cubic, leaveTangentWeight=0.2766, tangentMode=RCTM_User, time=0.863, value=2}
       ]}
     }}
    Scale Sprite Size Mode=::NewEnumerator3
   )
   &DynamicMaterialParameters(
    Index 0 Param 1=1
    Index 0 Param 2=1
    Index 0 Param 3=1.9518
    Index 0 Param 4=0.5547
    Module.Index0Float0Write=true
    Module.Index0Float1Write=true
    Module.Index0Float2Write=true
    Module.Index0Float3Write=true
    Module.Index1Float0Write=true
    Module.Index1Float1Write=true
    Module.Index1Float2Write=true
    Module.Index1Float3Write=true
    Module.Index2Float0Write=true
    Module.Index2Float1Write=true
    Module.Index2Float2Write=true
    Module.Index2Float3Write=true
    Module.Index3Float0Write=true
    Module.Index3Float1Write=true
    Module.Index3Float2Write=true
    Module.Index3Float3Write=true
    Write Parameter Index 1=::NewEnumerator1
    Write Parameter Index 2=::NewEnumerator1
    Write Parameter Index 3=::NewEnumerator1
   )
   &SpriteFacingAndAlignment(
    Sprite Alignment=&Lerp_Vector(
     A=&Multiply_Vector(A=Linked() {linkedVariable {name=Particles.Position, type=NiagaraPosition}}, B=<1 0 1>)
     B=<0 0 1>
     Alpha=&FloatFromCurve(
      FloatCurve=DataInterface() {propertyValues {Curve {keys [
          {interpMode=RCIM_Cubic, tangentMode=RCTM_User, time=0, value=0}
          {arriveTangent=2.5388, interpMode=RCIM_Cubic, leaveTangent=2.5388, tangentMode=RCTM_User, time=0.4178, value=1}
         ]}}}
      CurveIndex=$Emitter.Age
     )
    )
    WriteAlignment=true
    WriteFacing=false
   ) {module=SpriteFacingAndAlignment001}
  ]
  rendererClasses [&NiagaraSpriteRendererProperties]
  renderers [{rendererClass=&NiagaraSpriteRendererProperties}]
 }}
summary {
 systemName=NS_Emote_TrafficHat_Drops
 emitters [{emitterName=Minimal, rendererClasses [&NiagaraSpriteRendererProperties]}]
 userVariables [User.DropOffset {type=NiagaraFloat, defaultValue=0, description=""}]
}
