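# Niagara effect for the TrafficHat emote: a short purple-pink skull smoke burst.
# A curl-noise-driven Skull_Spline mesh with animated material parameters combines with toon smoke meshes, a mushroom cloud sprite and an impact flare.
# Burst-spawned droplets, fire bits and a brief light with an accompanying sprite add accents.
# Time-of-day functions tint or dim most layers, and a front-end modifier adjusts the skull mesh scale for menus versus gameplay.
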NiagaraSystem /BRCosmetics/Animation/Game/MainPlayer/Emotes/TrafficHat/FX/NS_Emote_TrafficHat_SkullSmoke
dependencies {
 usedDataInterfaces [
  &NiagaraDataInterfaceArrayColor
  &NiagaraDataInterfaceCurve
  &NiagaraDataInterfaceMeshRendererInfo
  &NiagaraDataInterfaceColorCurve
  &NiagaraDataInterfaceVectorCurve
  &NiagaraDataInterfaceArrayFloat3
 ]
 usedDynamicInputs [&Add_Float, &DotProduct, &ConvertVectorToPosition]
 usedModules [&SystemState]
}
emitters {
 Droplets {
  emitterUpdateScript [
   &EmitterState(Life Cycle Mode=::NewEnumerator1, Loop Behavior=::NewEnumerator1, Loop Duration=2, UseLoopDelay=false)
   &SpawnBurst_Instantaneous(
    Spawn Count=&RandomRangeInt(Minimum=4, Maximum=2)
    Spawn Time=0.05
    Use Spawn Probability=false
    UseLoopCountLimit=false
   )
  ]
  particleSpawnScript [
   &FN_SetSpawnGroup(Number of Spawn Groups=2)
   &InitializeParticle(
    Color=&SelectLinearColorFromArray(
     Array Sampling Mode=::NewEnumerator0
     Direct Array Index=?
     Color Selection Array=DataInterface() {propertyValues {ColorData [<rgba 0.6647 0.158 1 1>, <rgba 1 0.2201 0.9957 1>, <rgba 0.2956 0.1795 1 1>]}}
     Random Seed=0
    )
    Lifetime Max=0.4
    Lifetime Min=0.3
    Lifetime Mode=::NewEnumerator1
    Sprite Size=<2 4>
    Sprite Size Mode=::NewEnumerator4
    UsePositionOffset=true
    Write Material Random=true
   )
   &AddVelocity(
    Cone Angle=65
    Cone Axis=<0 0 1>
    Module.UseConeFalloff=false
    Module.UseVelDistribution=false
    Override Seed=false
    Velocity Mode=::NewEnumerator2
    Velocity Speed=&RandomRangeFloat(Minimum=250, Maximum=150)
   ) {module=AddVelocity001}
  ]
  particleUpdateScript [
   &ParticleState()
   &GravityForce(Gravity=<0 0 -285.5112>)
   &Drag(Use Linear Drag=true, Use Rotational Drag=true)
   &SolveForcesAndVelocity(Clamp Velocity=false, Limit Acceleration=false)
   &ScaleSpriteSizeBySpeed(Max Scale Factor=<1 10>)
   &ScaleSpriteSize(Uniform Curve Sprite Scale=DataInterface() {propertyValues {Curve {keys [
        {tangentMode=RCTM_User, time=0, value=1}
        {arriveTangent=-0.0405, interpMode=RCIM_Cubic, leaveTangent=-0.0405, tangentMode=RCTM_User, time=0.7009, value=1}
        {arriveTangent=-8.0089, interpMode=RCIM_Cubic, leaveTangent=-8.0089, tangentMode=RCTM_User, time=1, value=0}
       ]}}})
  ]
  rendererClasses [&NiagaraSpriteRendererProperties]
  renderers [{rendererClass=&NiagaraSpriteRendererProperties}]
 }
 FireBits {
  emitterUpdateScript [
   &EmitterState(Life Cycle Mode=::NewEnumerator1, Loop Behavior=::NewEnumerator1, Loop Duration=0.2, UseLoopDelay=true)
   &SpawnBurst_Instantaneous(Spawn Count=3, Spawn Time=0.07, Use Spawn Probability=false, UseLoopCountLimit=false)
   &SpawnBurst_Instantaneous(Spawn Group=1, Spawn Time=0.05, Use Spawn Probability=false, UseLoopCountLimit=false) {module=SpawnBurst_Instantaneous001}
  ]
  particleSpawnScript [
   /SetVariables(Particles.VisibilityTag=&MaskIntBySpawnGroup(Masked Int=0, Passthrough Int=1, Spawn Group Mask=0)) {isSetParameters=true}
   &InitializeParticle(
    Lifetime=&MaskFloatBySpawnGroup(Masked Float=&RandomRangeFloat(Minimum=1.5, Maximum=1), Passthrough Float=1)
    Sprite Rotation Angle=&MaskFloatBySpawnGroup(Masked Float=&RandomRangeFloat(Minimum=0, Maximum=1), Passthrough Float=-0.1422)
    Sprite Rotation Mode=::NewEnumerator1
    Sprite Size Mode=::NewEnumerator3
    Uniform Sprite Size=&MaskFloatBySpawnGroup(
     Masked Float=&RandomRangeFloat(Minimum=79, Maximum=83)
     Passthrough Float=&RandomRangeFloat(Minimum=38, Maximum=47)
    )
    UsePositionOffset=false
    Write Material Random=true
   )
   &DynamicMaterialParameters(
    Index 0 Param 2=&UniformRangedFloat(Minimum=0, Maximum=1)
    Index 0 Param 4=&UniformRangedFloat(Minimum=3, Maximum=5)
    Module.Index0Float0Write=true
    Module.Index0Float1Write=true
    Module.Index0Float2Write=true
    Module.Index0Float3Write=true
   )
   &ShapeLocation(
    Hemisphere Distribution=<1 1.6991>
    MaskSpawnGroup=false
    Non Uniform Scale=<1 1 1.6218>
    Offset Mode=::NewEnumerator0
    Override Seed=true
    Random Seed=4
    Sphere Radius=2.34
    Sphere Surface Distribution=0.3624
    UseOwnerScale=true
   )
   &AddVelocity(Velocity=&RandomRangeVector(Minimum=<0 0 104.2843>, Maximum=<1 1 83.4747>, Fixed Random Seed=true))
   &CameraOffset(Camera Offset Amount=&MaskFloatBySpawnGroup(Masked Float=-5, Passthrough Float=10))
  ]
  particleUpdateScript [
   &ParticleState()
   &ScaleSpriteSize(Scale Factor=&Vector2DFromFloat(Value=&FloatFromCurve(FloatCurve=DataInterface() {propertyValues {Curve {keys [
          {arriveTangent=0.0426, interpMode=RCIM_Cubic, leaveTangent=0.0426, tangentMode=RCTM_User, time=0.0002, value=0.15}
          {arriveTangent=0.5342, interpMode=RCIM_Cubic, leaveTangent=0.5342, tangentMode=RCTM_User, time=0.461, value=0.588}
          {arriveTangent=-0.1019, interpMode=RCIM_Cubic, leaveTangent=-0.1019, tangentMode=RCTM_User, time=1.0041, value=0.7174}
         ]}}})))
   &SpriteRotationRate(Rotation Rate=50)
   &Drag(
    Drag=&FloatFromCurve(
     FloatCurve=DataInterface() {propertyValues {Curve {keys [
         {arriveTangent=2.7494, interpMode=RCIM_Cubic, leaveTangent=2.7494, tangentMode=RCTM_User, time=0, value=0}
         {arriveTangent=1, interpMode=RCIM_Cubic, leaveTangent=1, time=0.2211, value=0.7761}
         {interpMode=RCIM_Cubic, tangentMode=RCTM_User, time=1, value=1}
        ]}}}
     Scale Curve=3.8444
    )
    Use Linear Drag=true
    Use Rotational Drag=true
   )
   &SolveForcesAndVelocity(Clamp Velocity=false, Limit Acceleration=false)
   /SetVariables(Particles.SubImageIndex=&FloatFromCurve(
     FloatCurve=DataInterface() {propertyValues {Curve {keys [{tangentMode=RCTM_User, time=0, value=0}, {tangentMode=RCTM_User, time=1, value=1}]}}}
     Scale Curve=40
    )) {isSetParameters=true}
   &DynamicMaterialParameters(
    Index 0 Param 1=&FloatFromCurve(FloatCurve=DataInterface() {propertyValues {Curve {keys [<at 0 0 RCIM_Cubic>, {arriveTangent=0.9989, interpMode=RCIM_Cubic, leaveTangent=0.9989, time=0.6, value=0.7}, <at 1.0011 1 RCIM_Cubic>]}}})
    Index 0 Param 2=2.5
    Index 0 Param 3=1
    Index 0 Param 4=3
    Module.Index0Float0Write=false
    Module.Index0Float1Write=true
    Module.Index0Float2Write=true
    Module.Index0Float3Write=true
   ) {module=DynamicMaterialParameters001}
   &ScaleColor(
    Scale Alpha=&FloatFromCurve(
     FloatCurve=DataInterface() {propertyValues {Curve {keys [{interpMode=RCIM_Cubic, tangentMode=RCTM_User, time=0.4955, value=1}, {interpMode=RCIM_Cubic, tangentMode=RCTM_User, time=1, value=0}]}}}
     CurveIndex=$Emitter.Age
     Scale Curve=&FNTimeOfDayScalar(Night Scale=0.5, Morning Scale=0.8, Evening Scale=0.8)
    )
    Scale RGB=&MakeVectorFromLinearColorRGB(Color=&FNTimeOfDayColor(
      Night Color=<rgba 0.3219 0.1259 0.6939 0.8>
      Morning Color=<rgba 0.4524 0.1482 0.6939 1>
      Day Color=<rgba 0.4524 0.1482 0.6939 1>
      Evening Color=<rgba 0.4524 0.1482 0.6939 1>
     ))
    ScaleA=true
    ScaleRGB=true
   ) {module=ScaleColor001}
   /SetVariables(Particles.BeakerTopPosition:NiagaraPosition=&AddVectorToPosition(Position=<0 0 0>, Vector=<0 0 6.8921>)) {isSetParameters=true}
   /SetVariables(Particles.BeakerTopColor:LinearColor=&FNTimeOfDayColor(
     Night Color=<rgba 0.3219 0.1259 0.6939 0.8>
     Morning Color=<rgba 0.4524 0.1482 0.6939 1>
     Day Color=<rgba 0.4524 0.1482 0.6939 1>
     Evening Color=<rgba 0.4524 0.1482 0.6939 1>
    )) {isSetParameters=true}
  ]
  rendererClasses [&NiagaraSpriteRendererProperties]
  renderers [
   {rendererClass=&NiagaraSpriteRendererProperties}
   {rendererClass=&NiagaraSpriteRendererProperties, rendererIndex=1}
  ]
 }
 Light {
  emitterUpdateScript [
   &EmitterState(Life Cycle Mode=::NewEnumerator1, Loop Behavior=::NewEnumerator1, Loop Delay=0.05, Loop Duration=1, UseLoopDelay=true)
   &SpawnBurst_Instantaneous(Use Spawn Probability=false, UseLoopCountLimit=false)
  ]
  particleSpawnScript [
   &FN_SetSpawnGroup(Number of Spawn Groups=2)
   &InitializeParticle(
    Color=<rgba 1 0.3503 0.8483 0.5>
    Lifetime=0.2
    Position Offset=<0 0 10>
    Sprite Size Mode=::NewEnumerator3
    Uniform Sprite Size=400.9189
    UsePositionOffset=true
    Write Material Random=true
   )
  ]
  particleUpdateScript [
   &ParticleState()
   &ScaleColor(
    Scale Alpha=&FloatFromCurve(FloatCurve=DataInterface() {propertyValues {Curve {keys [
         {interpMode=RCIM_Cubic, tangentMode=RCTM_User, time=0, value=1}
         {arriveTangent=-1, interpMode=RCIM_Cubic, leaveTangent=-1, time=0.2, value=1.475}
         {interpMode=RCIM_Cubic, tangentMode=RCTM_User, time=1, value=0}
        ]}}})
    Scale RGB=&Multiply_VectorByFloat(Vector=<1 1 1>, Float=&FNTimeOfDayScalar(Night Scale=0.3, Morning Scale=1, Evening Scale=1))
    ScaleA=true
    ScaleRGB=true
   )
   &FN_Light_Attributes(
    Color=$Particles.Color
    Create Color Attribute=true
    Create Exponent Attribute=true
    Create Radius Attribute=true
    Exponent=6
    Radius=400
    Scale Color=&FloatFromCurve(
     FloatCurve=DataInterface() {propertyValues {Curve {keys [
         {interpMode=RCIM_Cubic, tangentMode=RCTM_User, time=0, value=0}
         {
          arriveTangent=-0.0501
          arriveTangentWeight=0.348
          interpMode=RCIM_Cubic
          leaveTangent=-0.0501
          leaveTangentWeight=0.348
          tangentMode=RCTM_User
          tangentWeightMode=RCTWM_WeightedBoth
          time=0.5
          value=1
         }
         {tangentMode=RCTM_User, time=1, value=0}
        ]}}}
     Scale Curve=2
    )
   )
   &DynamicMaterialParameters(
    Index 1 Param 1=&MaskFloatBySpawnGroup(Masked Float=-2, Passthrough Float=1)
    Module.Index1Float0Write=true
    Module.Index1Float1Write=true
    Module.Index1Float2Write=true
    Module.Index1Float3Write=true
    Write Parameter Index 0=::NewEnumerator0
    Write Parameter Index 1=::NewEnumerator1
   )
   /SetVariables(Particles.DecalSize:Vector2f=<60 60>) {isSetParameters=true}
   /SetVariables(Particles.SpriteFacing=<0 -1 0>) {isSetParameters=true}
  ]
  rendererClasses [&NiagaraLightRendererProperties, &NiagaraSpriteRendererProperties]
  renderers [
   {rendererClass=&NiagaraLightRendererProperties, rendererIndex=0}
   {rendererClass=&NiagaraSpriteRendererProperties, rendererIndex=1}
  ]
 }
 MushroomCloud {
  emitterUpdateScript [
   &EmitterState(Life Cycle Mode=::NewEnumerator1, Loop Behavior=::NewEnumerator1, Loop Duration=1.406, UseLoopDelay=false)
   &SpawnBurst_Instantaneous(Spawn Count=1, Use Spawn Probability=false)
  ]
  particleSpawnScript [
   &InitializeParticle(
    Lifetime Max=0.5
    Lifetime Min=0.3
    Lifetime Mode=::NewEnumerator1
    Sprite Size=&Multiply_Vector2D_ByFloat(Float=1.3, Vector2D=&RandomRangeVector2D(Minimum=<70 90>, Maximum=<100 125>))
    Sprite Size Mode=::NewEnumerator4
    UsePositionOffset=false
    Write Material Random=true
   ) {module=InitializeParticle001}
   /SetVariables(Particles.Random=&RandomRangeFloat(Minimum=0, Maximum=1)) {isSetParameters=true}
   &AddVelocity() {enabled=false}
   &AddVelocity(Velocity=&MakeVector(X=0, Y=0, Z=59.1636)) {module=AddVelocity001}
   &CameraOffset(Camera Offset Amount=-20) {enabled=false}
   &DynamicMaterialParameters(
    Index 0 Param 2=&UniformAOrBFloat(A=0, Evaluation Type=::NewEnumerator0, Recalculate Random Each Loop=true, Override Seed=false)
    Module.Index0Float0Write=true
    Module.Index0Float1Write=true
    Module.Index0Float2Write=true
    Module.Index0Float3Write=true
   ) {module=DynamicMaterialParameters001}
  ]
  particleUpdateScript [
   &ParticleState()
   &ScaleSpriteSize(
    Scale Sprite Size Mode=::NewEnumerator0
    Uniform Scale Factor=&FloatFromCurve(FloatCurve=DataInterface() {propertyValues {Curve {keys [{interpMode=RCIM_Cubic, tangentMode=RCTM_User, time=0, value=0.4963}, {interpMode=RCIM_Cubic, tangentMode=RCTM_User, time=1, value=1}]}}})
   )
   &Drag(Drag=8, Use Linear Drag=true, Use Rotational Drag=false) {enabled=false}
   &Drag(
    Drag=&FloatFromCurve(
     FloatCurve=DataInterface() {propertyValues {Curve {keys [
         {arriveTangent=0.2819, interpMode=RCIM_Cubic, leaveTangent=0.2819, tangentMode=RCTM_User, time=0.05, value=1}
         {arriveTangent=14.4736, interpMode=RCIM_Cubic, leaveTangent=14.4736, tangentMode=RCTM_User, time=0.1, value=0}
         {arriveTangent=-1.1872, interpMode=RCIM_Cubic, leaveTangent=-1.1872, tangentMode=RCTM_User, time=0.8206, value=0.583}
        ]}}}
     Scale Curve=9.3364
    )
    Use Linear Drag=true
    Use Rotational Drag=false
   ) {module=Drag001}
   &SolveForcesAndVelocity(Clamp Velocity=false, Limit Acceleration=false)
   &DynamicMaterialParameters(
    Index 0 Param 1=&FloatFromCurve(FloatCurve=DataInterface() {propertyValues {Curve {keys [
         {arriveTangent=0.8348, interpMode=RCIM_Cubic, leaveTangent=0.8348, tangentMode=RCTM_User, time=0, value=0}
         {arriveTangent=2.1997, interpMode=RCIM_Cubic, leaveTangent=2.1997, tangentMode=RCTM_User, time=1, value=1}
        ]}}})
    Module.Index0Float0Write=true
    Module.Index0Float1Write=true
    Module.Index0Float2Write=true
    Module.Index0Float3Write=true
   )
   &ScaleColor(
    Scale Alpha=&Multiply_Float(
     A=&FNTimeOfDayScalar(Night Scale=1, Morning Scale=1, Evening Scale=1)
     B=&FloatFromCurve(FloatCurve=DataInterface() {propertyValues {Curve {keys [<at 0.5 1>,<at 1 0>]}}})
    )
    Scale RGB=&MakeVectorFromLinearColorRGB(Color=&FNTimeOfDayColor(
      Night Color=&ColorFromCurve(ColorCurve=DataInterface() {propertyValues {
         AlphaCurve {keys []}
         BlueCurve {keys [<at 0 0.8744 RCIM_Cubic>,<at 0.4534 0.8003 RCIM_Cubic>]}
         GreenCurve {keys [<at 0 0.5146 RCIM_Cubic>,<at 0.4534 0.1931 RCIM_Cubic>]}
         RedCurve {keys [<at 0 0.7855 RCIM_Cubic>,<at 0.4534 0.4819 RCIM_Cubic>]}
        }})
      Morning Color=&ColorFromCurve(ColorCurve=DataInterface() {propertyValues {
         AlphaCurve {keys []}
         BlueCurve {keys [<at 0 0.9217 RCIM_Cubic>,<at 1 0.9333 RCIM_Cubic>]}
         GreenCurve {keys [<at 0 0.283 RCIM_Cubic>,<at 1 0.3098 RCIM_Cubic>]}
         RedCurve {keys [<at 0 1 RCIM_Cubic>,<at 1 0.7412 RCIM_Cubic>]}
        }})
      Day Color=&ColorFromCurve(ColorCurve=DataInterface() {propertyValues {
         AlphaCurve {keys []}
         BlueCurve {keys [<at 0 0.9333 RCIM_Cubic>,<at 0.4639 1 RCIM_Cubic>]}
         GreenCurve {keys [<at 0 0.6921 RCIM_Cubic>,<at 0.4639 0.2196 RCIM_Cubic>]}
         RedCurve {keys [<at 0 0.8737 RCIM_Cubic>,<at 0.4639 0.8902 RCIM_Cubic>]}
        }})
      Evening Color=&ColorFromCurve(ColorCurve=DataInterface() {propertyValues {
         AlphaCurve {keys []}
         BlueCurve {keys [<at 0 0.9217 RCIM_Cubic>,<at 1 0.9333 RCIM_Cubic>]}
         GreenCurve {keys [<at 0 0.283 RCIM_Cubic>,<at 1 0.3098 RCIM_Cubic>]}
         RedCurve {keys [<at 0 1 RCIM_Cubic>,<at 1 0.7412 RCIM_Cubic>]}
        }})
     ))
    ScaleA=true
    ScaleRGB=true
   )
  ]
  rendererClasses [&NiagaraSpriteRendererProperties]
  renderers [{rendererClass=&NiagaraSpriteRendererProperties}]
 }
 Skull_Spline {
  emitterUpdateScript [
   &EmitterState(Life Cycle Mode=::NewEnumerator1, Loop Behavior=::NewEnumerator1, Loop Duration=1.5, UseLoopDelay=false)
   &SpawnBurst_Instantaneous(Spawn Time=0.03, Use Spawn Probability=false, UseLoopCountLimit=false)
  ]
  particleSpawnScript [
   &InitializeParticle(
    Lifetime=1.7
    Mesh Scale=&Multiply_VectorByFloat(Vector=<0.2399 0.182 0.96>, Float=0.6707)
    Mesh Scale Mode=::NewEnumerator4
    UsePositionOffset=false
    Write Material Random=true
   )
   &AddVelocity(
    Velocity=&RotateVectorByQuaternion(Vector To Rotate=<0 1 1>, Quaternion=$Particles.SkeletalMeshLocation.SampledRotation)
    Velocity Speed Scale=&RandomRangeFloat(Minimum=10, Maximum=20)
   )
  ]
  particleUpdateScript [
   &ParticleState()
   &ScaleMeshSize(Scale Factor=&MakeVector(
     X=&FN_FrontEndFloatModifier(Gameplay=1, Front End=0.6)
     Y=1
     Z=&FloatFromCurve(
      FloatCurve=DataInterface() {propertyValues {Curve {keys [
          {arriveTangent=9.0725, interpMode=RCIM_Cubic, leaveTangent=9.0725, tangentMode=RCTM_User, time=0, value=0}
          {arriveTangent=0.7067, interpMode=RCIM_Cubic, leaveTangent=0.7067, tangentMode=RCTM_User, time=0.017, value=0.974}
          {arriveTangent=0.5062, interpMode=RCIM_Cubic, leaveTangent=0.5062, time=0.7704, value=1.1212}
          {arriveTangent=0.6578, interpMode=RCIM_Cubic, leaveTangent=0.6578, tangentMode=RCTM_User, time=1, value=1.4716}
         ]}}}
      Scale Curve=&FN_FrontEndFloatModifier(Gameplay=0.85, Front End=0.7)
     )
    ))
   &CurlNoiseForce(
    Noise Strength=&FloatFromCurve(
     FloatCurve=DataInterface() {propertyValues {Curve {keys [{interpMode=RCIM_Cubic, tangentMode=RCTM_User, time=0, value=0}, {interpMode=RCIM_Cubic, tangentMode=RCTM_User, time=0.7, value=1}]}}}
     Scale Curve=&RandomRangeFloat(Minimum=25, Maximum=50)
    )
    Pan Noise=false
   )
   &Drag(Drag=0.5, Use Linear Drag=true, Use Rotational Drag=true)
   &SolveForcesAndVelocity(Clamp Velocity=false, Limit Acceleration=false)
   &DynamicMaterialParameters(
    Index 0 Param 1=&FloatFromCurve(FloatCurve=DataInterface() {propertyValues {Curve {keys [
         {arriveTangent=5.0399, interpMode=RCIM_Cubic, leaveTangent=5.0399, tangentMode=RCTM_User, time=0, value=-1}
         {arriveTangent=0.9837, interpMode=RCIM_Cubic, leaveTangent=0.9837, tangentMode=RCTM_User, time=0.1997, value=-0.2096}
         {arriveTangent=0.0502, interpMode=RCIM_Cubic, leaveTangent=0.0502, tangentMode=RCTM_User, time=0.6, value=-0.0272}
        ]}}})
    Index 0 Param 2=&FloatFromCurve(
     FloatCurve=DataInterface() {propertyValues {Curve {keys [
         {arriveTangent=0.0771, interpMode=RCIM_Cubic, leaveTangent=0.0771, tangentMode=RCTM_User, time=0.5768, value=0}
         {arriveTangent=2.1363, interpMode=RCIM_Cubic, leaveTangent=2.1363, tangentMode=RCTM_User, time=1, value=0.75}
        ]}}}
     Scale Curve=&RandomRangeFloat(Minimum=0.4, Maximum=0.6)
    )
    Index 0 Param 3=&FloatFromCurve(FloatCurve=DataInterface() {propertyValues {Curve {keys [
         {arriveTangent=-0.4297, interpMode=RCIM_Cubic, leaveTangent=-0.4297, tangentMode=RCTM_User, time=0, value=0.4549}
         {arriveTangent=-0.0708, interpMode=RCIM_Cubic, leaveTangent=-0.0708, tangentMode=RCTM_User, time=0.1638, value=0}
         {arriveTangent=0.0092, interpMode=RCIM_Cubic, leaveTangent=0.0092, tangentMode=RCTM_User, time=0.6109, value=0}
         {arriveTangent=3.7616, interpMode=RCIM_Cubic, leaveTangent=3.7616, tangentMode=RCTM_User, time=1, value=1}
        ]}}})
    Index 0 Param 4=&RandomRangeFloat(Minimum=0, Maximum=1)
    Index 1 Param 1=&FloatFromCurve(
     FloatCurve=DataInterface() {propertyValues {Curve {keys [
         {arriveTangent=0.0825, interpMode=RCIM_Cubic, leaveTangent=0.0825, tangentMode=RCTM_User, time=0.7079, value=0}
         {arriveTangent=2.7761, interpMode=RCIM_Cubic, leaveTangent=2.7761, tangentMode=RCTM_User, time=1, value=1}
        ]}}}
     Scale Curve=&RandomRangeFloat(Minimum=0.8, Maximum=1)
    )
    Index 1 Param 2=&FloatFromCurve(
     FloatCurve=DataInterface() {propertyValues {Curve {keys [
         {arriveTangent=-0.0388, interpMode=RCIM_Cubic, leaveTangent=-0.0388, tangentMode=RCTM_User, time=0, value=0}
         {arriveTangent=0.1591, interpMode=RCIM_Cubic, leaveTangent=0.1591, tangentMode=RCTM_User, time=0.3, value=1}
         {interpMode=RCIM_Cubic, tangentMode=RCTM_User, time=0.45, value=1}
         {arriveTangent=2.0344, interpMode=RCIM_Cubic, leaveTangent=2.0344, tangentMode=RCTM_User, time=1, value=2}
        ]}}}
     Scale Curve=0.6444
    )
    Index 2 Param 1=&FloatFromCurve(FloatCurve=DataInterface() {propertyValues {Curve {keys [
         {arriveTangent=0.0429, interpMode=RCIM_Cubic, leaveTangent=0.0429, tangentMode=RCTM_User, time=0.5026, value=0.3926}
         {interpMode=RCIM_Cubic, tangentMode=RCTM_User, time=1, value=0.8848}
        ]}}})
    Index 2 Param 2=&MakeFloatFromVector(Vector=$Particles.SkeletalMeshLocation.SampledPosition, Channel=::NewEnumerator1)
    Index 2 Param 3=&FloatFromCurve(
     FloatCurve=DataInterface() {propertyValues {Curve {keys [
         {interpMode=RCIM_Cubic, tangentMode=RCTM_User, time=0, value=1}
         {arriveTangent=-0.0336, interpMode=RCIM_Cubic, leaveTangent=-0.0336, tangentMode=RCTM_User, time=0.5, value=1}
         {interpMode=RCIM_Cubic, tangentMode=RCTM_User, time=1, value=0}
        ]}}}
     Scale Curve=&RandomRangeFloat(Minimum=0.8, Maximum=1.1)
    )
    Index 2 Param 4=&FloatFromCurve(FloatCurve=DataInterface() {propertyValues {Curve {keys [
         {interpMode=RCIM_Cubic, tangentMode=RCTM_User, time=0, value=1}
         {arriveTangent=0.6667, interpMode=RCIM_Cubic, leaveTangent=0.6667, time=0.2, value=0.55}
         <at 0.3 1.2 RCIM_Cubic>
         {arriveTangent=-0.0265, interpMode=RCIM_Cubic, leaveTangent=-0.0265, tangentMode=RCTM_User, time=0.4, value=0.55}
         {arriveTangent=0.0518, interpMode=RCIM_Cubic, leaveTangent=0.0518, tangentMode=RCTM_User, time=0.5, value=0.9}
         {interpMode=RCIM_Cubic, tangentMode=RCTM_User, time=0.8, value=0}
        ]}}})
    Module.Index0Float0Write=true
    Module.Index0Float1Write=true
    Module.Index0Float2Write=true
    Module.Index0Float3Write=true
    Module.Index1Float0Write=true
    Module.Index1Float1Write=true
    Module.Index1Float2Write=true
    Module.Index1Float3Write=true
    Module.Index2Float0Write=true
    Module.Index2Float1Write=true
    Module.Index2Float2Write=true
    Module.Index2Float3Write=true
    Write Parameter Index 1=::NewEnumerator1
    Write Parameter Index 2=::NewEnumerator1
   )
   &ScaleColor(
    Scale Alpha=&FloatFromCurve(
     FloatCurve=DataInterface() {propertyValues {Curve {keys [
         {interpMode=RCIM_Cubic, tangentMode=RCTM_User, time=0.8, value=1}
         {arriveTangent=-6.2584, interpMode=RCIM_Cubic, leaveTangent=-6.2584, tangentMode=RCTM_User, time=1, value=0}
        ]}}}
     Scale Curve=0.2
    )
    Scale RGB=&VectorFromFloat(Value=&FloatFromCurve(FloatCurve=DataInterface() {propertyValues {Curve {keys [
          {arriveTangent=19.0414, interpMode=RCIM_Cubic, leaveTangent=19.0414, tangentMode=RCTM_User, time=0, value=2.095}
          {interpMode=RCIM_Cubic, tangentMode=RCTM_User, time=0.3031, value=4}
          {arriveTangent=0.0291, interpMode=RCIM_Cubic, leaveTangent=0.0291, tangentMode=RCTM_User, time=0.6, value=4}
          {arriveTangent=-0.0639, interpMode=RCIM_Cubic, leaveTangent=-0.0639, tangentMode=RCTM_User, time=0.8, value=1}
         ]}}}))
    ScaleA=true
    ScaleRGB=true
   )
   &ScaleColor(
    Scale RGB=&MakeVectorFromLinearColorRGB(Color=&FNTimeOfDayColor(
      Night Color=&ColorFromCurve(ColorCurve=DataInterface() {propertyValues {
         AlphaCurve {keys []}
         BlueCurve {keys [<at 0 1.6007 RCIM_Cubic>,<at 0.134 5 RCIM_Cubic>]}
         GreenCurve {keys [<at 0 0.0809 RCIM_Cubic>,<at 0.134 0.7269 RCIM_Cubic>]}
         RedCurve {keys [<at 0 0.3147 RCIM_Cubic>,<at 0.134 2.3841 RCIM_Cubic>]}
        }})
      Morning Color=&ColorFromCurve(ColorCurve=DataInterface() {propertyValues {
         AlphaCurve {keys []}
         BlueCurve {keys [<at 0 1.6007 RCIM_Cubic>,<at 0.134 7 RCIM_Cubic>]}
         GreenCurve {keys [<at 0 0.0824 RCIM_Cubic>,<at 0.134 1.0901 RCIM_Cubic>]}
         RedCurve {keys [<at 0 0.537 RCIM_Cubic>,<at 0.134 5.6664 RCIM_Cubic>]}
        }})
      Day Color=&ColorFromCurve(ColorCurve=DataInterface() {propertyValues {
         AlphaCurve {keys []}
         BlueCurve {keys [<at 0 1.6007 RCIM_Cubic>,<at 0.134 10 RCIM_Cubic>]}
         GreenCurve {keys [<at 0 0.0824 RCIM_Cubic>,<at 0.134 1.5573 RCIM_Cubic>]}
         RedCurve {keys [<at 0 0.537 RCIM_Cubic>,<at 0.134 8.0948 RCIM_Cubic>]}
        }})
      Evening Color=&ColorFromCurve(ColorCurve=DataInterface() {propertyValues {
         AlphaCurve {keys []}
         BlueCurve {keys [<at 0 1.6007 RCIM_Cubic>,<at 0.134 7 RCIM_Cubic>]}
         GreenCurve {keys [<at 0 0.0824 RCIM_Cubic>,<at 0.134 1.0901 RCIM_Cubic>]}
         RedCurve {keys [<at 0 0.537 RCIM_Cubic>,<at 0.134 5.6664 RCIM_Cubic>]}
        }})
     ))
    ScaleA=true
    ScaleRGB=true
   ) {module=ScaleColor001}
  ]
  rendererClasses [&NiagaraMeshRendererProperties]
  renderers [{rendererClass=&NiagaraMeshRendererProperties, rendererIndex=0}]
 }
 ToonSmoke {
  emitterSpawnScript [/Game/Effects/Niagara/MasterMaterials/ToonShader/Template/NM_ToonShaderParameters(HighLight_Softness=0, Lighting_Hardness=0.8, Metallic_LightingPower=0, RimLight_Color:LinearColor=<rgba 0 0 0 1>, RimLight_Width=0)]
  emitterUpdateScript [
   &EmitterState(Life Cycle Mode=::NewEnumerator1, Loop Behavior=::NewEnumerator1, Loop Duration=1, UseLoopDelay=false)
   &SpawnBurst_Instantaneous(Spawn Count=5, Spawn Time=0.1, Use Spawn Probability=false, UseLoopCountLimit=false)
  ]
  particleSpawnScript [
   &InitializeParticle(
    Lifetime=&RandomRangeFloat(Minimum=0.6, Maximum=1.2)
    Mesh Scale Mode=::NewEnumerator1
    Mesh Uniform Scale Max=0.6
    Mesh Uniform Scale Min=0.8
    Sprite Size Mode=::NewEnumerator3
    Uniform Sprite Size=3.1789
    UsePositionOffset=false
    Write Material Random=true
   )
   &InitialMeshOrientation(Rotation=&RandomRangeVector(Minimum=<0 0 0>, Maximum=<1 1 1>), Use Rotation Vector=true)
   &ShapeLocation(
    MaskSpawnGroup=false
    Offset=<-10 -2 5>
    Offset Mode=::NewEnumerator0
    Rotation Angle=0.75
    Rotation Axis=<0 0 1>
    Rotation Mode=::NewEnumerator1
    Sphere Distribution=::NewEnumerator1
    Sphere Radius=10
    UseOwnerScale=true
   )
   &AddVelocity(Velocity=&Add_Vector(
     A=&SelectVectorFromArray(
      Direct Array Index=&ReturnExecIndex()
      Vector Selection Array=DataInterface() {propertyValues {FloatData [<120 120 130>, <-120 100 120>, <60 -40 150>, <-60 -120 100>, <-40 80 80>]}}
     )
     B=&VectorFromFloat(Value=&RandomRangeFloat(Minimum=0.2, Maximum=-0.2))
    ))
   &CameraOffset(Camera Offset Amount=&RandomRangeFloat(Minimum=0, Maximum=10))
  ]
  particleUpdateScript [
   &ParticleState()
   &ScaleMeshSize(Scale Factor=&VectorFromFloat(Value=&FloatFromCurve(FloatCurve=DataInterface() {propertyValues {Curve {keys [
          {arriveTangent=5.3301, interpMode=RCIM_Cubic, leaveTangent=5.3301, tangentMode=RCTM_User, time=0, value=0}
          {interpMode=RCIM_Cubic, tangentMode=RCTM_User, time=0.7, value=0.95}
          {tangentMode=RCTM_User, time=1, value=0}
         ]}}})))
   &Drag(Drag=&RandomRangeFloat(Minimum=9, Maximum=5), Use Linear Drag=true, Use Rotational Drag=false)
   &SolveForcesAndVelocity(Clamp Velocity=false, Limit Acceleration=false)
   &DynamicMaterialParameters(
    Index 0 Param 1=&FloatFromCurve(FloatCurve=DataInterface() {propertyValues {Curve {keys [
         {arriveTangent=4.1321, interpMode=RCIM_Cubic, leaveTangent=4.1321, tangentMode=RCTM_User, time=0.4084, value=0}
         {arriveTangent=1.43, interpMode=RCIM_Cubic, leaveTangent=1.43, tangentMode=RCTM_User, time=0.5122, value=0.7801}
         {arriveTangent=0.1245, interpMode=RCIM_Cubic, leaveTangent=0.1245, tangentMode=RCTM_User, time=1, value=1}
        ]}}})
    Module.Index0Float0Write=true
    Module.Index0Float1Write=true
    Module.Index0Float2Write=true
    Module.Index0Float3Write=true
   )
   &ScaleColor(
    Scale Alpha=&FloatFromCurve(FloatCurve=DataInterface() {propertyValues {Curve {keys [
         {arriveTangent=0.0022, interpMode=RCIM_Cubic, leaveTangent=0.0022, tangentMode=RCTM_User, time=0, value=1}
         {arriveTangent=0.1254, interpMode=RCIM_Cubic, tangentMode=RCTM_Break, time=1, value=0.4}
        ]}}})
    Scale RGB=&MakeVectorFromLinearColorRGB(Color=&FNTimeOfDayColor(
      Night Color=&ColorFromCurve(ColorCurve=DataInterface() {propertyValues {
         AlphaCurve {keys []}
         BlueCurve {keys [<at 0 0.8466 RCIM_Cubic>,<at 0.3275 0.5133 RCIM_Cubic>]}
         GreenCurve {keys [<at 0 0.3885 RCIM_Cubic>,<at 0.3275 0.1096 RCIM_Cubic>]}
         RedCurve {keys [<at 0 0.7334 RCIM_Cubic>,<at 0.3275 0.2377 RCIM_Cubic>]}
        }})
      Morning Color=&ColorFromCurve(ColorCurve=DataInterface() {propertyValues {
         AlphaCurve {keys []}
         BlueCurve {keys [<at 0 0.9217 RCIM_Cubic>,<at 1 0.9333 RCIM_Cubic>]}
         GreenCurve {keys [<at 0 0.283 RCIM_Cubic>,<at 1 0.3098 RCIM_Cubic>]}
         RedCurve {keys [<at 0 1 RCIM_Cubic>,<at 1 0.7412 RCIM_Cubic>]}
        }})
      Day Color=&ColorFromCurve(ColorCurve=DataInterface() {propertyValues {
         AlphaCurve {keys []}
         BlueCurve {keys [<at 0 0.9333 RCIM_Cubic>,<at 0.3275 1 RCIM_Cubic>]}
         GreenCurve {keys [<at 0 0.6921 RCIM_Cubic>,<at 0.3275 0.0538 RCIM_Cubic>]}
         RedCurve {keys [<at 0 0.8737 RCIM_Cubic>,<at 0.3275 0.4946 RCIM_Cubic>]}
        }})
      Evening Color=&ColorFromCurve(ColorCurve=DataInterface() {propertyValues {
         AlphaCurve {keys []}
         BlueCurve {keys [{arriveTangent=0.0116, interpMode=RCIM_Cubic, time=0, value=1}, <at 1 0.9333 RCIM_Cubic>]}
         GreenCurve {keys [{arriveTangent=0.0268, interpMode=RCIM_Cubic, time=0, value=0.114}, <at 1 0.0848 RCIM_Cubic>]}
         RedCurve {keys [{arriveTangent=-0.2588, interpMode=RCIM_Cubic, time=0, value=0.6026}, <at 1 0.2395 RCIM_Cubic>]}
        }})
     ))
    ScaleA=true
    ScaleRGB=true
   ) {module=ScaleColor001}
  ]
  rendererClasses [&NiagaraMeshRendererProperties]
  renderers [{rendererClass=&NiagaraMeshRendererProperties, rendererIndex=0}]
 }
 Toon_Impact_Flare {
  emitterUpdateScript [
   &EmitterState(Life Cycle Mode=::NewEnumerator1, Loop Behavior=::NewEnumerator1, Loop Duration=0.5, UseLoopDelay=false)
   &SpawnBurst_Instantaneous(Spawn Count=1, Spawn Time=0.04, Use Spawn Probability=false)
  ]
  particleSpawnScript [
   &InitializeParticle(
    Lifetime=0.3
    Mesh Scale=&Multiply_VectorByFloat(Vector=&RandomRangeVector(Minimum=<1.45 0.7557 0.7>, Maximum=<1.8524 0.6111 0.6025>), Float=0.2998)
    Mesh Scale Mode=::NewEnumerator4
    Position Offset=<0 0 2>
    Sprite Size=<200 175>
    Sprite Size Mode=::NewEnumerator4
    UsePositionOffset=true
    Write Material Random=true
   ) {module=InitializeParticle001}
   /SetVariables(Particles.CoronaSpriteSize=&Vector2DFromFloat(Value=90)) {isSetParameters=true}
   /SetVariables(Particles.ImpactColor:LinearColor=&MakeLinearColorFromVectorAndFloat("Vector (RGB)"=<1 1 1>, "Float (Alpha)"=&Multiply_Float(A=1, B=1))) {isSetParameters=true}
   &InitialMeshOrientation(Orientation Axis=<0 0 1>, Orientation Vector=<0 0 1>, Rotation=<0 -0.25 0>, Use Orientation Vector=true, Use Rotation Vector=true)
  ]
  particleUpdateScript [
   &CameraOffset(Camera Offset Amount=30) {enabled=false}
   &ParticleState()
   &ScaleMeshSize(Scale Factor=&VectorFromCurve(
     VectorCurve=DataInterface() {propertyValues {
       XCurve {keys [
         {arriveTangent=17.7413, interpMode=RCIM_Cubic, leaveTangent=17.7413, tangentMode=RCTM_User, time=0, value=1}
         {arriveTangent=2.0478, interpMode=RCIM_Cubic, leaveTangent=2.0478, tangentMode=RCTM_User, time=0.1687, value=2.6286}
         {arriveTangent=0.9077, interpMode=RCIM_Cubic, leaveTangent=0.9077, tangentMode=RCTM_User, time=1, value=3.7975}
        ]}
       YCurve {keys [<at 0 1>,<at 1 1.5>]}
       ZCurve {keys [{tangentMode=RCTM_User, time=0, value=1}, {tangentMode=RCTM_User, time=1, value=1.5}]}
      }}
     Scale Curve=<1.4334 1 1>
    )) {module=ScaleMeshSize001}
   &DynamicMaterialParameters(
    Index 0 Param 1=&FloatFromCurve(FloatCurve=DataInterface() {propertyValues {Curve {keys [<at 0 1 RCIM_Cubic>, {arriveTangent=-0.2645, interpMode=RCIM_Cubic, leaveTangent=-0.2645, tangentMode=RCTM_User, time=0.8, value=0}]}}})
    Index 0 Param 2=2
    Index 1 Param 1=&FloatFromCurve(FloatCurve=DataInterface() {propertyValues {Curve {keys [
         {arriveTangent=-2.3163, interpMode=RCIM_Cubic, time=0, value=2}
         {arriveTangent=-1.9019, interpMode=RCIM_Cubic, leaveTangent=-1.9019, tangentMode=RCTM_User, time=1, value=0.4}
        ]}}})
    Index 1 Param 2=-2.6
    Index 1 Param 3=0.6
    Index 1 Param 4=1
    Index 2 Param 1=&FloatFromCurve(FloatCurve=DataInterface() {propertyValues {Curve {keys [
         {interpMode=RCIM_Cubic, tangentMode=RCTM_User, time=0, value=0}
         {arriveTangent=1.7315, interpMode=RCIM_Cubic, leaveTangent=1.7315, tangentMode=RCTM_User, time=1, value=1}
        ]}}})
    Index 2 Param 2=45
    Module.Index0Float0Write=true
    Module.Index0Float1Write=true
    Module.Index0Float2Write=true
    Module.Index0Float3Write=true
    Module.Index1Float0Write=true
    Module.Index1Float1Write=true
    Module.Index1Float2Write=true
    Module.Index1Float3Write=true
    Module.Index2Float0Write=true
    Module.Index2Float1Write=true
    Module.Index2Float2Write=true
    Module.Index2Float3Write=true
    Write Parameter Index 1=::NewEnumerator1
    Write Parameter Index 2=::NewEnumerator1
   )
   /SetVariables(Particles.CoronaSpriteColor:LinearColor=&MakeLinearColorFromVectorAndFloat(
     "Vector (RGB)"=&MakeVectorFromLinearColorRGB(Color=<rgba 0.5521 0.5521 0.5521 1>)
     "Float (Alpha)"=&Multiply_Float(
      A=&FloatFromCurve(FloatCurve=DataInterface() {propertyValues {Curve {keys [<at 0.35 1.5 RCIM_Cubic>,<at 0.95 0>]}}})
      B=3.9
     )
    )) {isSetParameters=true}
   /SetVariables(Particles.ImpactColor:LinearColor=&FNTimeOfDayColor(
     Night Color=&ColorFromCurve(ColorCurve=DataInterface() {propertyValues {
        AlphaCurve {keys [<at 0 1 RCIM_Cubic>,<at 1 0 RCIM_Cubic>]}
        BlueCurve {keys [<at 0 0.8294 RCIM_Cubic>,<at 0.482 0.8559 RCIM_Cubic>]}
        GreenCurve {keys [<at 0 0.1811 RCIM_Cubic>,<at 0.482 0.1389 RCIM_Cubic>]}
        RedCurve {keys [<at 0 1 RCIM_Cubic>,<at 0.482 0.4373 RCIM_Cubic>]}
       }})
     Morning Color=&ColorFromCurve(ColorCurve=DataInterface() {propertyValues {
        AlphaCurve {keys [<at 0 1 RCIM_Cubic>,<at 1 0 RCIM_Cubic>]}
        BlueCurve {keys [<at 0 0.8294 RCIM_Cubic>,<at 0.482 0.8559 RCIM_Cubic>]}
        GreenCurve {keys [<at 0 0.1811 RCIM_Cubic>,<at 0.482 0.1389 RCIM_Cubic>]}
        RedCurve {keys [<at 0 1 RCIM_Cubic>,<at 0.482 0.4373 RCIM_Cubic>]}
       }})
     Day Color=&ColorFromCurve(ColorCurve=DataInterface() {propertyValues {
        AlphaCurve {keys [<at 0 1 RCIM_Cubic>,<at 1 0 RCIM_Cubic>]}
        BlueCurve {keys [<at 0 0.8294 RCIM_Cubic>,<at 0.482 0.8559 RCIM_Cubic>]}
        GreenCurve {keys [<at 0 0.1811 RCIM_Cubic>,<at 0.482 0.1389 RCIM_Cubic>]}
        RedCurve {keys [<at 0 1 RCIM_Cubic>,<at 0.482 0.4373 RCIM_Cubic>]}
       }})
     Evening Color=&ColorFromCurve(ColorCurve=DataInterface() {propertyValues {
        AlphaCurve {keys [<at 0 1 RCIM_Cubic>,<at 1 0 RCIM_Cubic>]}
        BlueCurve {keys [<at 0 0.8294 RCIM_Cubic>,<at 0.482 0.8559 RCIM_Cubic>]}
        GreenCurve {keys [<at 0 0.1811 RCIM_Cubic>,<at 0.482 0.1389 RCIM_Cubic>]}
        RedCurve {keys [<at 0 1 RCIM_Cubic>,<at 0.482 0.4373 RCIM_Cubic>]}
       }})
    )) {isSetParameters=true}
   &ScaleMeshSize(Scale Factor=&VectorFromFloat(Value=&FloatFromCurve(FloatCurve=DataInterface() {propertyValues {Curve {keys [{arriveTangent=1.0067, interpMode=RCIM_Cubic, leaveTangent=1.0067, tangentMode=RCTM_User, time=0, value=0.3933}, <at 1 1.202>]}}})))
  ]
  rendererClasses [&NiagaraMeshRendererProperties]
  renderers [{rendererClass=&NiagaraMeshRendererProperties, rendererIndex=0}]
 }
}
summary {
 systemName=NS_Emote_TrafficHat_SkullSmoke
 emitters [
  {emitterName=Droplets, rendererClasses [&NiagaraSpriteRendererProperties]}
  {emitterName=Skull_Spline, rendererClasses [&NiagaraMeshRendererProperties]}
  {emitterName=Toon_Impact_Flare, rendererClasses [&NiagaraMeshRendererProperties]}
  {emitterName=FireBits, rendererClasses [&NiagaraSpriteRendererProperties]}
  {emitterName=Light, rendererClasses [&NiagaraLightRendererProperties, &NiagaraSpriteRendererProperties]}
  {emitterName=MushroomCloud, rendererClasses [&NiagaraSpriteRendererProperties]}
  {emitterName=ToonSmoke, rendererClasses [&NiagaraMeshRendererProperties]}
 ]
 userVariables [
  User.bIsFrontEnd {type=NiagaraBool}
  User.bIsFrontEndPreview {type=NiagaraBool}
 ]
}
